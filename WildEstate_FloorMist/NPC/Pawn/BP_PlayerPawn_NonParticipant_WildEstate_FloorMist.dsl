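; Non-participant player pawn Blueprint for the Wild Estate Floor Mist content, which shows a rift portal effect when it spawns.
; On BeginPlay it runs the parent logic, then calls SpawnRiftEffect, which spawns the entry/exit portal Niagara system at the actor's location.
; The effect spawns only where cosmetic events can execute.
; Overlap and Tick handlers are disabled stubs, and construction just calls the parent.
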(event Custom|SpawnRiftEffect
  (Networking|CanExecuteCosmeticEvents
    (:True
      (Niagara|SpawnSystematLocation "/WildEstateBosses/Gameplay/RiftEffect/NS_WildEstate_Portal_EntryExit.NS_WildEstate_Portal_EntryExit" (Transformation|GetActorLocation)))))

(event EventBeginPlay
  (|CallParentFunction)
  (CallFunction|SpawnRiftEffect))

(event Collision|EventActorBeginOverlap
  (params (OtherActor :type Actor%20Object%20Reference))
  (|CallParentFunction OtherActor) ; This node is disabled and will not be called.\nDrag off pins to build functionality.
)

(event EventTick
  (params (DeltaSeconds :type Float%20%28single-precision%29))
  (|CallParentFunction DeltaSeconds) ; This node is disabled and will not be called.\nDrag off pins to build functionality.
)

(fn ConstructionScript
  (|CallParentFunction))
